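# Niagara trail effect for the BlowWire glider: red-orange wing, tip and socket glow ribbons plus mesh wind lines.
# Ribbon emitters burst fixed particle sets, then reposition them each update from skeletal mesh sockets, bending and fading them by ribbon index.
# System timelines and smoothed moving averages drive trail length and sway, with separate front-end and gameplay values.
# Wind lines spawn on a torus, steer with User.RightDot and scale alpha by distance.
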NiagaraSystem /BRCosmetics/Gadgets/Assets/VinderTech_GliderChute/Glider_BlowWire/FX/NS_Glider_BlowWire_Trail
dependencies {
 usedDataInterfaces [
  &NiagaraDataInterfaceArrayInt32
  &NiagaraDataInterfaceCurve
  &NiagaraDataInterfaceColorCurve
  &NiagaraDataInterfaceArrayFloat
  &NiagaraDataInterfaceArrayFloat3
  &NiagaraDataInterfaceSkeletalMesh
 ]
 usedDynamicInputs [
  &MakeIntFromFloat
  &NormalizeVector
  &TransformVector
 ]
 usedModules [
  &self:SystemSpawnScript.SystemScriptSource.SystemScriptGraph.NiagaraNodeAssignment_0.SetVariables_#1
  &SystemState
  &Timeline
  &FN_Integrator
  &InterpolateOverTime
 ]
}
emitters {
 NE_GlowTrails {
  emitterSpawnScript [
   /SetVariables(Emitter.RibbonNumber=4) {isSetParameters=true}
   /SetVariables(Emitter.ParticlePerRibbon=20) {isSetParameters=true}
  ]
  emitterUpdateScript [
   &EmitterState()
   &SpawnBurst_Instantaneous(
    Spawn Count=&Multiply_Int(A=$Emitter.RibbonNumber, B=$Emitter.ParticlePerRibbon)
    Use Spawn Probability=false
    UseLoopCountLimit=true
   )
  ]
  particleSpawnScript [
   &PartitionParticles(
    ID to Partition:NiagaraInt32=&ReturnExecIndex()
    Number Of Partitions=$Emitter.RibbonNumber
    Particles Per Partition:NiagaraInt32=$Emitter.ParticlePerRibbon
    Partition Mode:ENiagara_PartitionMode=::NewEnumerator1
    Write Partition Index to Ribbon Link Order:ENiagara_AttributeSamplingApplyOutput=::NewEnumerator0
    Write Partition to Ribbon ID=::NewEnumerator0
   )
   &InitializeParticle(
    Color Mode=::NewEnumerator0
    Lifetime=0.2
    Position Offset=<0 0 25>
    Ribbon Width=22
    Ribbon Width Mode=::NewEnumerator1
    Sprite Size Mode=::NewEnumerator3
    Uniform Sprite Size=&MakeCustomFloatFromBool(
     Bool=&SetBoolByFloatComparison(A=$Particles.PartitionParticles.Normalized Index In Partition, B=0.1, Comparison Type=::NewEnumerator6)
     True Float=&SelectFloatFromArray(Direct Array Index=$Particles.PartitionParticles.Partition, Float Selection Array=DataInterface() {propertyValues {FloatData [15,15,20,20]}})
     False Float=0
    )
    UsePositionOffset=false
    Write Material Random=true
   )
   /SetVariables(
    Particles.RibbonID=/Niagara/DynamicInputs/Meshes/MakeRibbonIDFromSpawnGroup()
    Particles.RibbonLinkOrder:NiagaraFloat=&ReturnNormalizedExecIndex()
    Particles.SpawnGroup=$Particles.PartitionParticles.Partition
   ) {isSetParameters=true, module=SetVariables_#2}
  ]
  particleUpdateScript [
   &ParticleState(Kill Particles When Lifetime Has Elapsed=false, Let Infinitely Lived Particles Die When Emitter Deactivates=true)
   &SkeletalMeshLocation(
    Bone / Socket Index=&SelectIntFromArray(Direct Array Index=$Particles.PartitionParticles.Partition, Int Selection Array=DataInterface() {propertyValues {IntData [4,5,6,7]}})
    MaskSpawnGroup=false
    Mesh Sampling Type=::NewEnumerator5
    Sampled Position Offset Coordinate Space:ENiagaraSkelMeshTransforms=::NewEnumerator1
    Skeletal Mesh=$System.Skeletal Mesh
    Socket Sampling Mode=::NewEnumerator5
   ) {module=SkeletalMeshLocation001}
   /SetVariables(Particles.Position=&AddVectorToPosition(
     Position=&LerpPosition(
      A=$Particles.Position
      B=&AddVectorToPosition(
       Position=$Particles.Position
       Vector=&RotateVectorByQuaternion(
        Vector To Rotate=&MakeVector(
         X=&FloatFromCurve(
          FloatCurve=DataInterface() {propertyValues {Curve {keys [<at 0 0 RCIM_Cubic>, {arriveTangent=2.1213, interpMode=RCIM_Cubic, leaveTangent=2.1213, tangentMode=RCTM_User, time=1, value=1}]}}}
          CurveIndex=$Particles.PartitionParticles.Normalized Index In Partition
          Scale Curve=&Multiply_Float(
           A=&MakeFloatFromVector(Vector=$System.InterpolateOverTime.InterpolateOverTime_V3_.Moving Average, Channel=::NewEnumerator0)
           B=&Lerp_Float(A=0, B=100, Alpha=$System.InterpolateOverTime001.InterpolateOverTime_Float.Moving Average)
          )
         )
         Y=&FloatFromCurve(
          FloatCurve=DataInterface() {propertyValues {Curve {keys [{tangentMode=RCTM_User, time=0, value=0}, {tangentMode=RCTM_User, time=1, value=1}]}}}
          CurveIndex=$Particles.PartitionParticles.Normalized Index In Partition
          Scale Curve=&Multiply_Float(
           A=&FN_FrontEndFloatModifier(
            Gameplay=&Lerp_Float(A=0, B=250, Alpha=$System.Timeline.NormalizedTime)
            Front End=&Lerp_Float(A=0, B=250, Alpha=$System.Timeline002.NormalizedTime)
           )
           B=&SelectFloatFromArray(Direct Array Index=$Particles.PartitionParticles.Partition, Float Selection Array=DataInterface() {propertyValues {FloatData [1,1,0.5,0.5]}})
          )
         )
         Z=0
        )
        Quaternion=&MakeQuatFromAxisAngle(Axis=<0 0 1>, AngleInDegrees=-90)
       )
      )
      Alpha=$Particles.PartitionParticles.Normalized Index In Partition
     )
     Vector=<0 0 0>
    )) {isSetParameters=true, module=SetVariables_#3}
   &ScaleColor(
    Curve Index=$Particles.PartitionParticles.Normalized Index In Partition
    Linear Color Curve=DataInterface() {propertyValues {
      AlphaCurve {keys [<at 0.33 1>,<at 0.9269 1>,<at 1 0>]}
      BlueCurve {keys [<at 0 0>,<at 0.5521 0>,<at 0.9957 0>]}
      GreenCurve {keys [<at 0 3.1497>,<at 0.5521 0>,<at 0.9957 0.0042>]}
      RedCurve {keys [<at 0 10>,<at 0.5521 5>,<at 0.9957 0.0781>]}
     }}
    Scale Mode=::NewEnumerator2
   )
   &ScaleColor(
    Scale Alpha=&FloatFromCurve(
     FloatCurve=DataInterface() {propertyValues {Curve {keys [{tangentMode=RCTM_User, time=0, value=1}, {tangentMode=RCTM_User, time=1, value=0}]}}}
     CurveIndex=$Particles.PartitionParticles.Normalized Index In Partition
     Scale Curve=&FN_FrontEndFloatModifier(
      Gameplay=&Lerp_Float(A=0, B=0.5, Alpha=$System.Timeline.NormalizedTime)
      Front End=&Lerp_Float(A=0, B=0.5, Alpha=$System.Timeline002.NormalizedTime)
     )
    )
    ScaleA=true
    ScaleRGB=true
   ) {module=ScaleColor001}
   /SetVariables(Particles.GlowColor=&Multiply_LinearColorByFloat(Linear Color=$Particles.Color, Float=0.3)) {isSetParameters=true}
   /SetVariables(Particles.GlowPos:NiagaraPosition=&AddVectorToPosition(
     Position=$Particles.SkeletalMeshLocation001.SampledPosition
     Vector=&SelectVectorFromArray(
      Direct Array Index=$Particles.PartitionParticles.Partition
      Vector Selection Array=DataInterface() {propertyValues {FloatData [<4 0 0>, <4 0 0>, <7 0 0>, <7 0 0>]}}
     )
    )) {isSetParameters=true}
   /SetVariables(Particles.GlowTrailColor:LinearColor=&MultiplyLinearColor(Linear Color=$Particles.Color, Scale Factor=<1 1 1 0.15>)) {isSetParameters=true}
   /SetVariables(Particles.GlowTrailWidth:NiagaraFloat=&Multiply_Float(A=$Particles.RibbonWidth, B=3)) {isSetParameters=true}
   &ScaleRibbonWidth(Ribbon Width Scale=&SelectFloatFromArray(
     Direct Array Index=$Particles.PartitionParticles.Partition
     Float Selection Array=DataInterface() {propertyValues {FloatData [0.2,0.2,0.3,0.3]}}
    ))
  ]
  rendererClasses [&NiagaraRibbonRendererProperties, &NiagaraSpriteRendererProperties]
  renderers [
   {rendererClass=&NiagaraRibbonRendererProperties}
   {rendererClass=&NiagaraSpriteRendererProperties, rendererIndex=1}
   {rendererClass=&NiagaraRibbonRendererProperties, rendererIndex=2}
  ]
 }
 NE_TipTrails {
  emitterUpdateScript [
   &EmitterState()
   &SpawnBurst_Instantaneous(Spawn Count=30, Use Spawn Probability=false, UseLoopCountLimit=true)
  ]
  particleSpawnScript [
   &PartitionParticles(ID to Partition:NiagaraInt32=&ReturnExecIndex(), Write Partition to Ribbon ID=::NewEnumerator0)
   &InitializeParticle(
    Color Mode=::NewEnumerator0
    Lifetime=0.2
    Position Offset=<0 0 25>
    Ribbon Width=22
    Ribbon Width Mode=::NewEnumerator1
    Sprite Size Mode=::NewEnumerator3
    Uniform Sprite Size=5
    UsePositionOffset=false
    Write Material Random=true
   )
   /SetVariables(
    Particles.RibbonID=&MakeRibbonIDFromSpawnGroup()
    Particles.RibbonLinkOrder:NiagaraFloat=&ReturnNormalizedExecIndex()
    Particles.SpawnGroup=$Particles.PartitionParticles.Partition
   ) {isSetParameters=true, module=SetVariables_#2}
  ]
  particleUpdateScript [
   &ParticleState(Kill Particles When Lifetime Has Elapsed=false, Let Infinitely Lived Particles Die When Emitter Deactivates=true)
   &SkeletalMeshLocation(
    Bone / Socket Index=&SelectIntFromArray(Direct Array Index=$Particles.PartitionParticles.Partition, Int Selection Array=DataInterface() {propertyValues {IntData [0,1,0,0]}})
    MaskSpawnGroup=false
    Mesh Sampling Type=::NewEnumerator5
    Sampled Position Offset Coordinate Space:ENiagaraSkelMeshTransforms=::NewEnumerator1
    Skeletal Mesh=$System.Skeletal Mesh
    Socket Sampling Mode=::NewEnumerator5
   ) {module=SkeletalMeshLocation001}
   &self:SetRibbonBetweenTwoPoints(
    FirstSocket:NiagaraInt32=&SelectIntFromArray(Direct Array Index=$Particles.PartitionParticles.Partition, Int Selection Array=DataInterface() {propertyValues {IntData [2,0]}})
    RibbonWidth=0.05
    SecondSocket:NiagaraInt32=&SelectIntFromArray(Direct Array Index=$Particles.PartitionParticles.Partition, Int Selection Array=DataInterface() {propertyValues {IntData [3,1]}})
    Skeletal Mesh:NiagaraDataInterfaceSkeletalMesh=$System.Skeletal Mesh
   ) {module=SetRibbonBetweenTwoPoints}
   /SetVariables(Particles.Position=&LerpPosition(
     A=$Particles.Position
     B=&AddVectorToPosition(
      Position=$Particles.Position
      Vector=&RotateVectorByQuaternion(
       Vector To Rotate=&MakeVector(
        X=&FloatFromCurve(
         FloatCurve=DataInterface() {propertyValues {Curve {keys [<at 0 0 RCIM_Cubic>, {arriveTangent=2.1213, interpMode=RCIM_Cubic, leaveTangent=2.1213, tangentMode=RCTM_User, time=1, value=1}]}}}
         CurveIndex=&ReturnNormalizedExecIndex()
         Scale Curve=&Multiply_Float(
          A=&MakeFloatFromVector(Vector=$System.InterpolateOverTime.InterpolateOverTime_V3_.Moving Average, Channel=::NewEnumerator0)
          B=&Lerp_Float(A=0, B=100, Alpha=$System.InterpolateOverTime001.InterpolateOverTime_Float.Moving Average)
         )
        )
        Y=&FloatFromCurve(
         FloatCurve=DataInterface() {propertyValues {Curve {keys [{tangentMode=RCTM_User, time=0, value=0}, {tangentMode=RCTM_User, time=1, value=1}]}}}
         CurveIndex=&ReturnNormalizedExecIndex()
         Scale Curve=&FN_FrontEndFloatModifier(
          Gameplay=&Lerp_Float(A=0, B=250, Alpha=$System.Timeline.NormalizedTime)
          Front End=&Lerp_Float(A=0, B=250, Alpha=$System.Timeline002.NormalizedTime)
         )
        )
        Z=0
       )
       Quaternion=&MakeQuatFromAxisAngle(Axis=<0 0 1>, AngleInDegrees=-90)
      )
     )
     Alpha=&ReturnNormalizedExecIndex()
    )) {isSetParameters=true, module=SetVariables_#3}
   &ScaleColor(
    Curve Index=$Particles.RibbonLinkOrder
    Linear Color Curve=DataInterface() {propertyValues {
      AlphaCurve {keys [<at 0 4>,<at 0.9269 1>,<at 1 0>]}
      BlueCurve {keys [<at 0 0.1>,<at 0.6371 0>,<at 0.9957 0>]}
      GreenCurve {keys [<at 0 0>,<at 0.6371 0>,<at 0.9957 0.0042>]}
      RedCurve {keys [<at 0 1>,<at 0.6371 0.7656>,<at 0.9957 0.0781>]}
     }}
    Scale Mode=::NewEnumerator2
   )
   &ScaleColor(
    Scale Alpha=&FloatFromCurve(
     FloatCurve=DataInterface() {propertyValues {Curve {keys [{tangentMode=RCTM_User, time=0, value=1}, {tangentMode=RCTM_User, time=1, value=0}]}}}
     CurveIndex=$Particles.Initial.RibbonLinkOrder
     Scale Curve=&FN_FrontEndFloatModifier(
      Gameplay=&Lerp_Float(A=0, B=0.6, Alpha=$System.Timeline.NormalizedTime)
      Front End=&Lerp_Float(A=0, B=0.7, Alpha=$System.Timeline002.NormalizedTime)
     )
    )
    ScaleA=true
    ScaleRGB=true
   ) {module=ScaleColor001}
  ]
  rendererClasses [&NiagaraRibbonRendererProperties]
  renderers [{rendererClass=&NiagaraRibbonRendererProperties}]
 }
 NE_WingTrails {
  emitterUpdateScript [
   &EmitterState()
   &SpawnBurst_Instantaneous(Spawn Count=30, Use Spawn Probability=false, UseLoopCountLimit=true)
  ]
  particleSpawnScript [
   &PartitionParticles(ID to Partition:NiagaraInt32=&ReturnExecIndex(), Write Partition to Ribbon ID=::NewEnumerator0)
   &InitializeParticle(
    Color Mode=::NewEnumerator0
    Lifetime=0.2
    Position Offset=<0 0 25>
    Ribbon Width=22
    Ribbon Width Mode=::NewEnumerator1
    Sprite Size Mode=::NewEnumerator3
    Uniform Sprite Size=5
    UsePositionOffset=false
    Write Material Random=true
   )
   /SetVariables(
    Particles.RibbonID=&MakeRibbonIDFromSpawnGroup()
    Particles.RibbonLinkOrder:NiagaraFloat=&ReturnNormalizedExecIndex()
    Particles.SpawnGroup=$Particles.PartitionParticles.Partition
   ) {isSetParameters=true, module=SetVariables_#2}
  ]
  particleUpdateScript [
   &ParticleState(Kill Particles When Lifetime Has Elapsed=false, Let Infinitely Lived Particles Die When Emitter Deactivates=true)
   &SkeletalMeshLocation(
    Bone / Socket Index=&SelectIntFromArray(Direct Array Index=$Particles.PartitionParticles.Partition, Int Selection Array=DataInterface() {propertyValues {IntData [0,1,0,0]}})
    MaskSpawnGroup=false
    Mesh Sampling Type=::NewEnumerator5
    Sampled Position Offset Coordinate Space:ENiagaraSkelMeshTransforms=::NewEnumerator1
    Skeletal Mesh=$System.Skeletal Mesh
    Socket Sampling Mode=::NewEnumerator5
   ) {module=SkeletalMeshLocation001}
   &self:SetRibbonBetweenTwoPoints(
    FirstSocket:NiagaraInt32=&SelectIntFromArray(Direct Array Index=$Particles.PartitionParticles.Partition, Int Selection Array=DataInterface() {propertyValues {IntData [2,0]}})
    MidPointAlpha=0.5
    RibbonWidth=1
    SecondSocket:NiagaraInt32=&SelectIntFromArray(Direct Array Index=$Particles.PartitionParticles.Partition, Int Selection Array=DataInterface() {propertyValues {IntData [3,1]}})
    Skeletal Mesh:NiagaraDataInterfaceSkeletalMesh=$System.Skeletal Mesh
   ) {module=SetRibbonBetweenTwoPoints}
   /SetVariables(Particles.Position=&LerpPosition(
     A=$Particles.Position
     B=&AddVectorToPosition(
      Position=$Particles.Position
      Vector=&RotateVectorByQuaternion(
       Vector To Rotate=&MakeVector(
        X=&FloatFromCurve(
         FloatCurve=DataInterface() {propertyValues {Curve {keys [<at 0 0 RCIM_Cubic>, {arriveTangent=2.1213, interpMode=RCIM_Cubic, leaveTangent=2.1213, tangentMode=RCTM_User, time=1, value=1}]}}}
         CurveIndex=&ReturnNormalizedExecIndex()
         Scale Curve=&Multiply_Float(
          A=&MakeFloatFromVector(Vector=$System.InterpolateOverTime.InterpolateOverTime_V3_.Moving Average, Channel=::NewEnumerator0)
          B=&Lerp_Float(A=0, B=100, Alpha=$System.InterpolateOverTime001.InterpolateOverTime_Float.Moving Average)
         )
        )
        Y=&FloatFromCurve(
         FloatCurve=DataInterface() {propertyValues {Curve {keys [{tangentMode=RCTM_User, time=0, value=0}, {tangentMode=RCTM_User, time=1, value=1}]}}}
         CurveIndex=&ReturnNormalizedExecIndex()
         Scale Curve=&FN_FrontEndFloatModifier(
          Gameplay=&Lerp_Float(A=0, B=250, Alpha=$System.Timeline.NormalizedTime)
          Front End=&Lerp_Float(A=0, B=250, Alpha=$System.Timeline002.NormalizedTime)
         )
        )
        Z=0
       )
       Quaternion=&MakeQuatFromAxisAngle(Axis=<0 0 1>, AngleInDegrees=-90)
      )
     )
     Alpha=&ReturnNormalizedExecIndex()
    )) {isSetParameters=true, module=SetVariables_#3}
   &ScaleColor(
    Curve Index=$Particles.RibbonLinkOrder
    Linear Color Curve=DataInterface() {propertyValues {
      AlphaCurve {keys [<at 0.33 1>,<at 0.9269 1>,<at 1 0>]}
      BlueCurve {keys [<at 0 0.05>,<at 0.4702 0>,<at 0.9957 0>]}
      GreenCurve {keys [<at 0 0>,<at 0.4702 0>,<at 0.9957 0.0042>]}
      RedCurve {keys [<at 0 1>,<at 0.4702 0.7>,<at 0.9957 0.0781>]}
     }}
    Scale Mode=::NewEnumerator2
   )
   &ScaleColor(
    Scale Alpha=&FloatFromCurve(
     FloatCurve=DataInterface() {propertyValues {Curve {keys [{tangentMode=RCTM_User, time=0, value=1}, {tangentMode=RCTM_User, time=1, value=0}]}}}
     CurveIndex=$Particles.Initial.RibbonLinkOrder
     Scale Curve=&FN_FrontEndFloatModifier(
      Gameplay=&Lerp_Float(A=0, B=0.4, Alpha=$System.Timeline.NormalizedTime)
      Front End=&Lerp_Float(A=0, B=0.35, Alpha=$System.Timeline002.NormalizedTime)
     )
    )
    ScaleA=true
    ScaleRGB=true
   ) {module=ScaleColor001}
  ]
  rendererClasses [&NiagaraRibbonRendererProperties]
  renderers [{rendererClass=&NiagaraRibbonRendererProperties}]
 }
 WindLines {
  emitterUpdateScript [
   &EmitterState()
   &SpawnRate(
    SpawnRate=&Lerp_Float(
     A=50
     B=150
     Alpha=&RemapRange(Input Value To Remap=$User.RightDot, Input Min=-1, Input Max=1, Output Min=0, Output Max=1)
    )
    Use Spawn Probability=true
   ) {enabled=false}
   &SpawnRate(SpawnGroup=1, SpawnRate=200, Use Spawn Probability=true) {module=SpawnRate001}
  ]
  particleSpawnScript [
   &InitializeParticle(
    Color=<rgba 1 1 1 4>
    Mesh Scale Max=<25 2 2>
    Mesh Scale Min=<10 1 1>
    Mesh Scale Mode=::NewEnumerator2
    UsePositionOffset=false
    Write Material Random=true
   )
   &InitialMeshOrientation(
    Mesh Orientation Mode=::NewEnumerator2
    Rotation=&RandomRangeVector(Minimum=<0 0 0>, Maximum=<1 0 0>)
    Use Rotation Vector=true
   )
   &TorusLocation(
    Align Torus With Axis=true
    Handle Radius=0
    Large Radius:NiagaraFloat=&RandomRangeFloat(Minimum=80, Maximum=150)
    MaskSpawnGroup=false
    Offset=<0 -180 0>
    Torus Axis=<0 1 0>
    Torus Distribution Mode:ENiagaraTorusDistributionMode=::NewEnumerator1
    Torus Origin:NiagaraPosition=&Add_Vector(
     A=Linked() {linkedVariable {name=Particles.Position, type=NiagaraPosition}}
     B=&MakeVector(X=0, Y=&MakeCustomFloatFromBool(Bool=$User.bIsFrontEndPreview, True Float=200, False Float=0), Z=0)
    )
    U Position:NiagaraFloat=&MakeCustomFloatFromBool(
     Bool=&SetBoolByFloatComparison(A=$User.RightDot, B=0, Comparison Type=::NewEnumerator4)
     True Float=&RandomRangeFloat(Minimum=0, Maximum=1)
     False Float=&Lerp_Float(
      A=&RandomRangeFloat(Minimum=0.25, Maximum=0.75)
      B=&RandomRangeFloat(Minimum=-0.25, Maximum=0.25)
      Alpha=&RemapRange(Input Value To Remap=$User.RightDot, Input Min=-1, Input Max=1, Output Min=0, Output Max=1)
     )
    )
    V Position:NiagaraFloat=&RandomRangeFloat(Minimum=0, Maximum=1)
   ) {enabled=false}
   &TorusLocation(
    Align Torus With Axis=true
    Handle Radius=0
    Large Radius:NiagaraFloat=&RandomRangeFloat(Minimum=100, Maximum=120)
    MaskSpawnGroup=false
    Module.Override Randomness=false
    Offset=<-180 0 0>
    Spawn Group Mask=1
    Torus Axis=<1 0 0>
    Torus Origin:NiagaraPosition=&Add_Vector(A=<-100 0 0>, B=&Multiply_VectorByFloat(Vector=<0 200 0>, Float=$User.RightDot))
   ) {module=TorusLocation001}
   &AddVelocityFromPoint(
    Module.Use Falloff Distance=false
    Origin Offset=&Add_Vector(A=<-500 0 0>, B=&Lerp_Vector(A=<0 0 0>, B=<0 300 0>, Alpha=$User.RightDot))
    Velocity Strength=150
   )
   &AddVelocity(
    Module.Use Falloff Distance=false
    Origin Offset=&Add_Vector(A=<-500 0 0>, B=&Multiply_VectorByFloat(Vector=<0 500 1>, Float=$User.RightDot))
    Override Seed=false
    Velocity Mode=::NewEnumerator1
    Velocity Speed=750
   ) {module=AddVelocity001}
   &AddVelocity(Velocity=&Multiply_Vector(
     A=&RandomRangeVector(Minimum=<200 0 1>, Maximum=<300 0 1>)
     B=&MakeVector(X=&MakeCustomFloatFromBool(Bool=$User.bIsFrontEndPreview, True Float=-1, False Float=1), Y=0, Z=1)
    ))
  ]
  particleUpdateScript [
   &ScaleVelocity(Velocity Scale=&VectorFromFloat(Value=&FloatFromCurve(
      FloatCurve=DataInterface() {propertyValues {Curve {keys [
          {interpMode=RCIM_Cubic, tangentMode=RCTM_User, time=0, value=0}
          {arriveTangent=2, interpMode=RCIM_Cubic, leaveTangent=2, time=0.1, value=0.975}
          {interpMode=RCIM_Cubic, tangentMode=RCTM_User, time=0.5, value=1}
          {arriveTangent=-2, interpMode=RCIM_Cubic, leaveTangent=-2, time=0.9, value=1}
          {tangentMode=RCTM_User, time=1, value=0}
         ]}}}
      CurveIndex=&RemapRange(Input Value To Remap=$User.RightDot, Input Min=-1, Input Max=1, Output Min=0, Output Max=1)
     )))
   &SolveForcesAndVelocity(Clamp Velocity=false, Limit Acceleration=false)
   &ParticleState()
   &ScaleMeshSize(Scale Factor=&VectorFromFloat(Value=&FloatFromCurve(FloatCurve=DataInterface() {propertyValues {Curve {keys [
          {tangentMode=RCTM_User, time=0, value=1}
          {arriveTangent=-0.0405, interpMode=RCIM_Cubic, leaveTangent=-0.0405, tangentMode=RCTM_User, time=0.7009, value=1}
          {arriveTangent=-8.0089, interpMode=RCIM_Cubic, leaveTangent=-8.0089, tangentMode=RCTM_User, time=1, value=0}
         ]}}})))
   &DynamicMaterialParameters(
    Index 0 Param 2=1
    Index 0 Param 3=1
    Index 0 Param 4=1
    Module.Param0WriteEnabled=true
    Module.Param1WriteEnabled=true
    Module.Param2WriteEnabled=true
    Module.Param3WriteEnabled=true
   )
   &ScaleColor(
    Scale Alpha=&Multiply_Float(
     A=&FloatFromCurve(
      FloatCurve=DataInterface() {propertyValues {Curve {keys [
          {interpMode=RCIM_Cubic, tangentMode=RCTM_User, time=0, value=0}
          {interpMode=RCIM_Cubic, tangentMode=RCTM_User, time=0.5, value=1.5}
          {tangentMode=RCTM_User, time=1, value=0}
         ]}}}
      Scale Curve=&Lerp_Float(
       A=&FN_FrontEndFloatModifier(Gameplay=0, Front End=0.3)
       B=&Lerp_Float(A=0, B=1, Alpha=$System.Timeline001.Time)
       Alpha=$System.Timeline.Time
      )
     )
     B=&FN_LOD_Distance_Remap(Start Distance=350, End Distance=750, Minimum Output=0, Maximum Output=1, Clamp Output=true)
    )
    ScaleA=true
    ScaleRGB=true
   )
  ]
  rendererClasses [&NiagaraMeshRendererProperties]
  renderers [{rendererClass=&NiagaraMeshRendererProperties, rendererIndex=0}]
 }
}
summary {
 systemName=NS_Glider_BlowWire_Trail
 emitters [
  {emitterName=NE_WingTrails, rendererClasses [&NiagaraRibbonRendererProperties]}
  {emitterName=WindLines, rendererClasses [&NiagaraMeshRendererProperties]}
  {emitterName=NE_TipTrails, rendererClasses [&NiagaraRibbonRendererProperties]}
  {
   emitterName=NE_GlowTrails
   rendererClasses [&NiagaraRibbonRendererProperties, &NiagaraSpriteRendererProperties]
  }
 ]
 userVariables [
  User.bIsFrontEnd {type=NiagaraBool}
  User.bIsFrontEndPreview {type=NiagaraBool}
  User.bIsFullyDeployed {type=NiagaraBool}
  User.ForwardDot {type=NiagaraFloat, defaultValue=1, description=""}
  User.RibbonIDOffset {type=NiagaraInt32}
  User.RightDot {type=NiagaraFloat, defaultValue=0, description=""}
 ]
}
